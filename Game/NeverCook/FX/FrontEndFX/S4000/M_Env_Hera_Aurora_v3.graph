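# Additive unlit aurora curtain material for Niagara mesh particles, in front-end FX.
# WPO scales wall height and adds sine/cosine waves, driven by particle dynamic parameters.
# Two panning textures plus a bottom stripe index curve-atlas color ramps, with hue, desaturation, speed, height and opacity varied by vertex color blue.
# Opacity combines that pattern with UV edge gradients and distance fades, using depth fade only on higher-quality, non-mobile deferred paths.
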Material /Game/NeverCook/FX/FrontEndFX/S4000/M_Env_Hera_Aurora_v3
graph {
 properties {
  bCastRayTracedShadows=false
  bUsedWithNiagaraMeshParticles=true
  blendMode=BLEND_Additive
  pixelDepthOffsetMode=PDOM_Legacy
  shadingModel=MSM_Unlit
  usedShadingModels=Unlit
 }
 nodes [
  LinearInterpolate(
   A=ScalarParameter() {defaultValue=0.03, group=WPO, parameterName=Pan Speed Rand Min, sliderMax=1, sortPriority=10}
   B=ScalarParameter() {defaultValue=0.05, group=WPO, parameterName=Pan Speed Rand Max, sliderMax=1, sortPriority=11}
   Alpha=VertexColor().B
  ) {constB=0.04}
  n1 = Divide(
   A=NamedRerouteUsage()
   B=Max(A=ScalarParameter() {defaultValue=0.3, group=Edge Grads, parameterName=Left Grad Spread, sliderMax=1, sortPriority=10}) {constB=0.0001}
  )
  n2 = Divide(
   A=NamedRerouteUsage()
   B=Max(A=ScalarParameter() {defaultValue=0.3, group=Edge Grads, parameterName=Right Grad Spread, sliderMax=1, sortPriority=11}) {constB=0.0001}
  )
  n3 = Power(
   Base=Multiply(A=Saturate(None=@n1), B=Saturate(None=@n2))
   Exp=ScalarParameter() {defaultValue=1, group=Edge Grads, parameterName=Left Right Grad Exponent}
  ) {constExponent=5}
  n4 = Power(
   Base=NamedRerouteUsage()
   Exp=ScalarParameter() {defaultValue=1, group=Edge Grads, parameterName=Top Grad Exponent, sliderMax=10, sliderMin=1, sortPriority=20}
  ) {constExponent=1}
  n5 = OneMinus(None=Power(Base=NamedRerouteUsage()) {constExponent=3})
  n6 = ScalarParameter() {defaultValue=2, group=Edge Grads, parameterName=Bottom Grad Exponent, sliderMax=10, sliderMin=1, sortPriority=21}
  n7 = Power(Base=@n5, Exp=@n6) {constExponent=3}
  n8 = Multiply(
   A=Multiply(A=NamedRerouteDeclaration(None=@n3) {name=Left Right Grads, nodeColor=<rgba 0 1 0 1>})
   B=Multiply(A=OneMinus(None=@n4), B=NamedRerouteDeclaration(None=@n7) {name=Bottom Grad, nodeColor=<rgba 0 1 0 1>})
  ) {constA=1}
  Multiply(A=@n8, B=NamedRerouteUsage())
  n9 = LinearInterpolate(
   A=VectorParameter() {group=WPO, parameterName=Min Wall Scaling, sortPriority=0, defaultValue=<rgba 0 0 8000 1>}.RGB
   B=VectorParameter() {group=WPO, parameterName=Max Wall Scaling, sortPriority=1, defaultValue=<rgba 0 0 18836.4102 1>}.RGB
   Alpha=VertexColor().B
  )
  n10 = Multiply(
   A=ScalarParameter() {defaultValue=1, group=WPO, parameterName=Wall Scalar, sortPriority=-1}
   B=DynamicParameter() {defaultValue=<rgba 0.4 1 1 1>, paramNames [SineSpeed, LeftRightAmplitude, UpDownAplitude, WallHeight]}.WallHeight
  )
  n11 = ScalarParameter() {defaultValue=0.5, group=WPO, parameterName=Left Right Min Taper, sortPriority=40}
  n12 = ScalarParameter() {defaultValue=1, group=WPO, parameterName=Left Right Max Taper, sortPriority=41}
  n13 = Multiply(
   A=Power(Base=NamedRerouteUsage(), Exp=ScalarParameter() {defaultValue=1, group=WPO, parameterName=Left Right Taper Exponent, sortPriority=42})
   B=NamedRerouteUsage()
  )
  n14 = Transform(None=Multiply(A=Multiply(A=@n9, B=@n10), B=LinearInterpolate(A=@n11, B=@n12, Alpha=@n13))) {transformSourceType=TRANSFORMSOURCE_Local}
  n15 = ComponentMask(None=TextureCoordinate()) {g=true}
  n16 = DynamicParameter() {defaultValue=<rgba 0.4 1 1 1>, paramNames [SineSpeed, LeftRightAmplitude, UpDownAplitude, WallHeight]}
  n17 = ScalarParameter() {defaultValue=1, group=WPO, parameterName=Left Right Amplitude Mult, sortPriority=30}
  n18 = LinearInterpolate(A=VertexNormalWS(), B=Transform(None=Constant3Vector() {constant=<rgba 0 1 0 0>}))
  n19 = NamedRerouteUsage()
  n20 = VectorParameter() {
   group=WPO
   parameterName="Sine Panning && Frequency"
   channelNames {a=Up Down Panning, b=Up Down Freq, g=Left Right Panning, r=Left Right Freq}
   defaultValue=<rgba 1 1 1 1>
  }
  n21 = Multiply(A=@n19, B=@n20.Left Right Freq)
  n22 = Multiply(
   A=Time()
   B=DynamicParameter() {defaultValue=<rgba 0.4 1 1 1>, paramNames [SineSpeed, LeftRightAmplitude, UpDownAplitude, WallHeight]}.SineSpeed
  )
  n23 = Multiply(A=@n20.Left Right Panning, B=@n22)
  n24 = Cosine(None=Add(A=Add(A=@n21, B=@n23), B=ScalarParameter() {group=WPO, parameterName=Cosine Offset}))
  n25 = Multiply(
   A=DynamicParameter() {defaultValue=<rgba 0.4 1 1 1>, paramNames [SineSpeed, LeftRightAmplitude, UpDownAplitude, WallHeight]}.UpDownAplitude
   B=ScalarParameter() {defaultValue=1, group=WPO, parameterName=Top Down Amplitude Mult, sortPriority=31}
  )
  n26 = Add(A=Multiply(A=@n19, B=@n20.Up Down Freq), B=Multiply(A=@n20.Up Down Panning, B=@n22))
  n27 = ScalarParameter() {group=WPO, parameterName=Sine Offset}
  n28 = Multiply(A=@n18, B=Sine(None=Add(A=@n26, B=@n27)))
  n29 = Sine(None=ComponentMask(None=TextureCoordinate() {uTiling=0.5, vTiling=0.5}) {r=true})
  n30 = Add(
   A=Multiply(A=Multiply(A=@n16.LeftRightAmplitude, B=@n17), B=Multiply(A=@n18, B=@n24)) {constB=20}
   B=Multiply(A=Multiply(A=@n25, B=@n28) {constB=20}, B=Saturate(None=@n29)) {constB=20}
  )
  NamedRerouteDeclaration(None=Add(A=Multiply(A=@n14, B=@n15), B=NamedRerouteDeclaration(None=@n30) {name=SineWave, nodeColor=<rgba 1 0.7031 0 1>})) {name=WPO, nodeColor=<rgba 0 0 1 1>}
  n31 = ComponentMask(None=TextureCoordinate()) {r=true}
  n32 = CurveAtlasRowParameter(CurveTime=LinearInterpolate(A=@n31, B=OneMinus(None=@n31), Alpha=ScalarParameter() {group=Curve, parameterName=InvetColorRamp})) {
   atlas=/Game/Effects/Fort_Effects/Effects/Environment/Hera/IceBiome/Auroras/CA_Env_Hera_Aurora
   curve=/Game/Effects/Fort_Effects/Effects/Environment/Hera/IceBiome/Auroras/C_Env_Hera_Aurora
   parameterName=AuroraColor
  }
  n33 = ScalarParameter() {defaultValue=0.5, group=Curve, parameterName=InvertColorRamp Amount}
  n34 = ComponentMask(None=OneMinus(None=TextureCoordinate())) {g=true}
  n35 = TextureCoordinate() {uTiling=7, vTiling=0.1}
  n36 = Constant() {r=0.5}
  n37 = Multiply(
   A=Time()
   B=LinearInterpolate(
    A=ScalarParameter() {group=Time, parameterName=Time Scale Min, sortPriority=0}
    B=ScalarParameter() {defaultValue=1, group=Time, parameterName=Time Scale Max, sortPriority=1}
    Alpha=VertexColor().B
   ) {constB=0.7}
  )
  n38 = TextureSample(UVs=Panner(
    Coordinate=MaterialFunctionCall(UVs=@n35, "Rotation Angle (0-1)"=@n36) {materialFunction=&CustomRotator}.Rotated Values
    Time=@n37
   ) {speedX=0.2}) {texture=/Game/Effects/Fort_Effects/Textures/Power_Source/T_Power_Source_Sphere_Texture}
  n39 = CurveAtlasRowParameter(CurveTime=Add(A=@n34, B=Multiply(A=@n38.R) {constB=0.05})) {
   atlas=/Game/Athena/LinearColorCurves/VFX_CurveAtlas
   curve=/Game/Accessories/FORT_Backpacks/Backpack_F_MED_Quarrel/FX/Curve_Quarrel_Auroras_03
   defaultValue=30
   group=Curve
   parameterName=MaskRamp2
  }
  n40 = ScalarParameter() {defaultValue=0.7, parameterName=BottomGlow}
  n41 = MaterialFunctionCall(UVs=TextureCoordinate() {uTiling=2, vTiling=0.1}, "Rotation Angle (0-1)"=@n36) {materialFunction=&CustomRotator}
  n42 = Multiply(
   A=@n38.R
   B=TextureSample(UVs=Panner(Coordinate=@n41.Rotated Values, Time=@n37) {speedX=-0.2}) {texture=&T_Inky_Smoke_Tile}.R
  )
  n43 = ScalarParameter() {defaultValue=2, parameterName=TexMultiply}
  n44 = VertexColor()
  n45 = Add(
   A=NamedRerouteDeclaration(None=Multiply(A=@n39.A, B=@n40) {constB=0.5}) {name=Bottom Stripe, nodeColor=<rgba 0.8984 1 0 1>}
   B=Multiply(A=Multiply(A=@n34, B=@n42), B=LinearInterpolate(B=@n43, Alpha=@n44.B) {constA=1.5}) {constB=5}
  )
  n46 = ScalarParameter() {defaultValue=0.2, parameterName=ColorAdd}
  n47 = Add(B=Saturate(None=Add(A=@n45, B=@n46) {constB=0.05}))
  n48 = CurveAtlasRowParameter(CurveTime=@n47) {
   atlas=&CA_Env_Hera_Aurora
   curve=&C_Env_Hera_Aurora
   parameterName=AuroraColor
  }
  n49 = MaterialFunctionCall(
   Hue Shift Percentage=LinearInterpolate(
    A=ScalarParameter() {group=Color, parameterName=Rand Hue Shift Min, sortPriority=60}
    B=ScalarParameter() {defaultValue=0.05, group=Color, parameterName=Rand Hue Shift Max, sortPriority=61}
    Alpha=VertexColor().B
   )
   Texture=Add(A=Multiply(A=@n32, B=@n33) {constB=0.5}, B=@n48)
  ) {materialFunction=&HueShift}
  n50 = LinearInterpolate(
   A=ScalarParameter() {group=Color, parameterName=Rand Desat Min, sortPriority=62}
   B=ScalarParameter() {defaultValue=0.2, group=Color, parameterName=Rand Desat Max, sortPriority=63}
   Alpha=VertexColor().B
  )
  NamedRerouteDeclaration(None=MaterialFunctionCall(Emissive=Desaturation(None=@n49.Result, Fraction=@n50)) {materialFunction=&MF_FX_EmissiveAdjustments}.Result) {name=Color, nodeColor=<rgba 1 0 0 1>}
  n51 = TextureCoordinate()
  n52 = ComponentMask(None=@n51) {r=true}
  NamedRerouteDeclaration(None=OneMinus(None=NamedRerouteDeclaration(None=@n52) {name=TexU, nodeColor=<rgba 1 0 1 1>})) {name=TexU Inverted, nodeColor=<rgba 1 0 1 1>}
  n53 = ComponentMask(None=@n51) {g=true}
  NamedRerouteDeclaration(None=OneMinus(None=NamedRerouteDeclaration(None=@n53) {name=TexV, nodeColor=<rgba 1 0 1 1>})) {name=TexV Inverted, nodeColor=<rgba 1 0 1 1>}
  n54 = ScalarParameter() {defaultValue=1000, group=Opacity, parameterName=Near Fade Length}
  n55 = ScalarParameter() {defaultValue=25000, group=CloseFarFader, parameterName=Far Fade Length, sortPriority=2}
  NamedRerouteDeclaration(None=MaterialFunctionCall!(
    NearFadeLength=@n54
    NearFadeOffset=ScalarParameter() {defaultValue=1, group=Opacity, parameterName=Near Fade Offset}
    NearFadeCurve
    NearFadeStartingScale
    NearFadeEndingScale
    FarFadeLength=@n55
    FarFadeOffset=ScalarParameter() {defaultValue=12500, group=CloseFarFader, parameterName=Far Fade Offset, sortPriority=3}
    FarFadeCurve
    FarFadeEndingScale
    Opacity Near Distance Bias
    Opacity Far Distance Bias=ScalarParameter() {group=CloseFarFader, parameterName="Opacity Far Distance Bias (S)", sortPriority=4}
   ) {materialFunction=/Game/Effects/Fort_Effects/Materials/_Functions/MF_SmoothLod}.Opacity) {name=CloseFarFade, nodeColor=<rgba 0.2656 0 1 1>}
  n56 = NamedRerouteUsage()
  n57 = Multiply(A=Abs(None=Subtract(A=@n56) {constB=0.5})) {constB=2}
  n58 = Power(Base=NamedRerouteUsage(), Exp=ScalarParameter() {defaultValue=2, parameterName=Taper Power, sliderMax=10})
  n59 = ScalarParameter() {defaultValue=1, parameterName=Taper Amount, sliderMax=5}
  n60 = Saturate(None=OneMinus(None=Multiply(A=@n58, B=@n59)))
  NamedRerouteDeclaration(None=OneMinus(None=Multiply(A=@n57, B=@n60))) {name=Taper Edges, nodeColor=<rgba 1 0.7266 0 1>}
  n61 = Multiply(A=NamedRerouteUsage()) {constB=2}
  n62 = Multiply(A=Multiply(A=@n47, B=@n8), B=Saturate(None=Add(A=@n5, B=@n61)))
  n63 = DepthFade(Opacity=@n62, FadeDistance=ScalarParameter() {group=Opacity, parameterName=Depth Fade Distance, sortPriority=6})
  n64 = QualitySwitch(Default=@n63, Low=@n62, High=@n63, Medium=@n62, Epic=@n63)
  n65 = ShadingPathSwitch(Default=@n64, Deferred=@n64, Forward=@n62, Mobile=@n62)
  n66 = NamedRerouteUsage()
  n67 = LinearInterpolate(
   A=ScalarParameter() {defaultValue=0.5, group=Opacity, parameterName=Min Opacity Scalar, sortPriority=1}
   B=ScalarParameter() {defaultValue=1, group=Opacity, parameterName=Max Opacity Scalar, sortPriority=2}
   Alpha=VertexColor().B
  )
  n68 = ScalarParameter() {defaultValue=1, group=Opacity, parameterName=MinMax Opacity Scalar Exponent, sortPriority=3}
  NamedRerouteDeclaration(None=Multiply(A=Multiply(A=@n65, B=@n66), B=Power(Base=@n67, Exp=@n68))) {name=Opacity, nodeColor=<rgba 0 1 0 1>}
  n69 = ScalarParameter() {defaultValue=2, parameterName=Far Fade Length Scalar}
  NamedRerouteDeclaration(None=MaterialFunctionCall(NearFadeLength=@n54, FarFadeOffset=Multiply(A=@n55, B=@n69)) {materialFunction=&MF_SmoothLod}.Opacity) {name=CloseFarScale, nodeColor=<rgba 0.2656 0 1 1>}
  NamedRerouteDeclaration(None=OneMinus(None=Fresnel() {baseReflectFraction=0, exponent=1})) {name=Fresnel, nodeColor=<rgba 0.6875 1 0 1>}
  NamedRerouteUsage()
  NamedRerouteUsage()
  NamedRerouteUsage()
  NamedRerouteUsage()
  NamedRerouteUsage()
  ScalarParameter() {defaultValue=25000, group=CloseFarFader, parameterName=Far Fade Length TopDown}
  n70 = CameraPositionWS()
  n71 = ObjectPositionWS()
  SmoothStep(
   Min=ScalarParameter() {parameterName=Min Height Top Fade}
   Max=ScalarParameter() {defaultValue=80000, parameterName=Max Height Top Fade}
   Value=Subtract(A=ComponentMask(None=@n70) {b=true}, B=ComponentMask(None=@n71) {b=true})
  )
 ]
}
